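# Physics asset for the WartyBrine glider cosmetic's skeletal mesh in the GliderChute gadget assets.
# Eight single-capsule bodies cover the frame, both wings, the handle hub and a base and handle on each side.
# Constraints form a tree rooted at Frame: the wings and Handles_Base attach to it, and each side's Handle_Base and Handle chain off Handles_Base.
# Every joint shares the same symmetric swing and twist limits.
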PhysicsAsset /BRCosmetics/Gadgets/Assets/VinderTech_GliderChute/Glider_WartyBrine/Meshes/WartyBrine_Glider_PhysicsAsset
bodies {
 Frame [{shapeType=Capsule, length=106.0953, radius=69.3743, center=<3.5267 -14.8962 -7.9902>, rotation=<rot 0 -175.9194 -87.3911>}]
 Handle_Base_l [{shapeType=Capsule, length=43.2273, radius=15.695, center=<-3.5847 -8.885 25.5322>, rotation=<rot -18.8337 -90 35.8591>}]
 Handle_Base_r [{shapeType=Capsule, length=41.8103, radius=14.5317, center=<4.5054 10.002 -28.0678>, rotation=<rot -18.309 -90 34.5626>}]
 Handle_l [{shapeType=Capsule, length=51.8974, radius=31.5153, center=<-24.5695 -9.9778 -23.3032>, rotation=<rot 0 135.1144 -88.5056>}]
 Handle_r [{shapeType=Capsule, length=52.4586, radius=34.534, center=<28.91 8.4404 23.7277>, rotation=<rot 0 130.0093 -88.3971>}]
 Handles_Base [{shapeType=Capsule, length=58.6475, radius=33.0227, center=<-2.5326 -10.9884 1.1138>, rotation=<rot 0 87.7776 -88.7585>}]
 Wing_l [{shapeType=Capsule, length=78.1166, radius=22.3607, center=<-60.863 -0.5082 22.1721>, rotation=<rot 0 177.5612 -85.0923>}]
 Wing_r [{shapeType=Capsule, length=82.3923, radius=21.2613, center=<59.9574 -8.3283 -21.5545>, rotation=<rot 0 178.8991 -85.9766>}]
}
constraints [
 {bone1Name=Wing_l, bone2Name=Frame, swing1LimitDegrees=45, swing2LimitDegrees=45, twistLimitDegrees=45}
 {bone1Name=Wing_r, bone2Name=Frame, swing1LimitDegrees=45, swing2LimitDegrees=45, twistLimitDegrees=45}
 {bone1Name=Handles_Base, bone2Name=Frame, swing1LimitDegrees=45, swing2LimitDegrees=45, twistLimitDegrees=45}
 {bone1Name=Handle_Base_r, bone2Name=Handles_Base, swing1LimitDegrees=45, swing2LimitDegrees=45, twistLimitDegrees=45}
 {bone1Name=Handle_r, bone2Name=Handle_Base_r, swing1LimitDegrees=45, swing2LimitDegrees=45, twistLimitDegrees=45}
 {bone1Name=Handle_Base_l, bone2Name=Handles_Base, swing1LimitDegrees=45, swing2LimitDegrees=45, twistLimitDegrees=45}
 {bone1Name=Handle_l, bone2Name=Handle_Base_l, swing1LimitDegrees=45, swing2LimitDegrees=45, twistLimitDegrees=45}
]
